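# Sound cue for a candle flame ambient prop loop, layering a steady flicker with occasional sizzles.
# One branch randomly picks a looping flicker wave, with slight start delay and fade-in envelope.
# A quieter branch loops delayed, modulated one-shot sizzles from a weighted random of three.
# The mix feeds only the first two quality slots, with short-range filtered attenuation, listener focus, low priority and tight concurrency.
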SoundCue /Game/Sounds/Helios/Environment_Actors/CandleFlame/Spot_Fire_Candle_Flame_Loop_Cue
root=QualityLevel_0
nodes [
 {id=Delay_0, children [Enveloper_1], properties {delayMax=0.25}}
 {id=Delay_1, children [Modulator_3], properties {delayMax=5, delayMin=3}}
 {
  id=Enveloper_1
  children [Random_1]
  properties {pitchCurve {editorCurveData {keys [{time=0}]}}, volumeCurve {editorCurveData {keys [<at 0 0>,<at 0.25 1>]}}}
 }
 {id=Looping_0, children [Delay_1]}
 {id=Mixer_0, children [WavePlayer_8], properties {inputVolume [0.25]}}
 {id=Mixer_1, children [Delay_0,Looping_0], properties {inputVolume [2,0.5]}}
 {id=Mixer_2, children [WavePlayer_7], properties {inputVolume [0.25]}}
 {id=Mixer_3, children [WavePlayer_9], properties {inputVolume [3]}}
 {id=Modulator_3, children [Random_2], properties {pitchMax=1, pitchMin=0.9, volumeMax=1.1}}
 {id=QualityLevel_0, children [Mixer_1,Mixer_1,null,null]}
 {id=Random_1, children [WavePlayer_4,WavePlayer_5], properties {weights [1,1]}}
 {id=Random_2, children [Mixer_2,Mixer_0,Mixer_3], properties {weights [0.2,0.2,1]}}
 {
  id=WavePlayer_4
  properties {bLooping=true, soundWaveAssetPtr=&/Spot_Fire_CandleFlicker_Loop_02}
 }
 {
  id=WavePlayer_5
  properties {bLooping=true, soundWaveAssetPtr=&/Spot_Fire_CandleFlicker_Loop_01}
 }
 {id=WavePlayer_7, properties {soundWaveAssetPtr=&/Spot_Fire_CandleSizzle_Swt_01}}
 {id=WavePlayer_8, properties {soundWaveAssetPtr=&/Spot_Fire_CandleSizzle_Swt_02}}
 {id=WavePlayer_9, properties {soundWaveAssetPtr=&/Spot_Fire_CandleSizzle_Swt_03}}
]
properties {
 bBypassVolumeScaleForPriority=true
 bOverrideAttenuation=true
 bOverrideConcurrency=true
 maxDistance=270
 priority=0
 soundClassObject=&Fort_Ambient_Props_Athena
 volumeMultiplier=1.002
 attenuationOverrides {
  bAttenuateWithLPF=true
  bEnableFocusInterpolation=true
  bEnableListenerFocus=true
  distanceAlgorithm=NaturalSound
  falloffDistance=250
  hPFFrequencyAtMax=450
  hPFFrequencyAtMin=80
  lPFRadiusMax=450
  lPFRadiusMin=100
  nonFocusVolumeAttenuation=0.75
  reverbDistanceMax=350
  reverbDistanceMin=100
  stereoSpread=0
  attenuationShapeExtents {x=20}
 }
 concurrencyOverrides {maxCount=2, voiceStealReleaseTime=2}
 soundSubmixSends [{minSendDistance=500, minSendLevel=0.6, soundSubmix=&Fortnite_Reverb_Default}]
}
